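# Parent material for the TriadSpear pickaxe skin, adding an animated audio-equalizer glow to the 2023 character parent shading.
# Frontend audio-reactivity amplitude, with panning-noise alternatives selected by If switches, drives distorted red and blue line masks and offset bar textures.
# Cylindrically projected noise varies the colors and motion, and the effect joins emissive between the PreFX and PostFX functions.
# A temporal responsiveness output is also written from effect masks.
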Material /BRCosmetics/Weapons/FORT_Melee/Pickaxe_ElegantHeist/Skins/TriadSpear/Materials/M_Pickaxe_TriadSpear_Parent
graph {
 properties {
  bAutomaticallySetUsageInEditor=false
  bUseMaterialAttributes=true
  bUsedWithMeshDeformer=true
  bUsedWithMorphTargets=true
  bUsedWithSkeletalMesh=true
  pixelDepthOffsetMode=PDOM_Legacy
  shadingModel=MSM_FromMaterialExpression
  usedShadingModels=Default Lit | Subsurface | Clear Coat | Subsurface Profile | Cloth
 }
 nodes [
  n1 = MaterialFunctionCall() {materialFunction=&MF_CharacterParent_2023_PreFX}
  n2 = GetMaterialAttributes(None=@n1.Result) {attributeGetTypes [#1]}
  n3 = NamedRerouteUsage()
  n4 = NamedRerouteUsage()
  MaterialFunctionCall(
   Material=SetMaterialAttributes!(MaterialAttributes=@n2.MaterialAttributes, Emissive Color=Add(A=@n3, B=@n4), World Position Offset) {attributeSetTypes [#1, #2]}
   M=@n1.M
   Pre Skinned Local Position=@n1.M
   Distance Blend=@n1.M
   Fresnel=@n1.M
  ) {materialFunction=&MF_CharacterParent_2023_PostFX}
  n5 = NamedRerouteUsage()
  Multiply(A=@n5) {constB=1.2}
  n6 = Add(A=NamedRerouteUsage()) {constB=0.02}
  Multiply(A=@n6, B=Constant() {r=1.5}) {constB=3}
  n7 = Constant() {r=0.5}
  n8 = MaterialFunctionCall!(Tile=Constant() {r=1}, DistortionAdjustment=Constant() {r=-7.1828}) {materialFunction=/BRCosmetics/Characters/Player/Female/Medium/Bodies/F_MED_TriadSpear/Materials/Body/MF_CylindricalProjUV}
  n9 = TextureSample!(UVs=Panner(Coordinate=@n8.UVs) {speedY=0.2}, Tex, "DDX(UVs)"=@n8.DDX, "DDY(UVs)"=@n8.DDX, Apply View MipBias) {
   mipValueMode=TMVM_Derivative
   samplerSource=SSM_Wrap_WorldGroupSettings
   samplerType=SAMPLERTYPE_Grayscale
   texture=/Game/Global/Textures/Default/Noise/LowResBlurredNoiseNormalized
  }
  n10 = Saturate(None=Power(Base=Multiply(A=@n9.R) {constB=2}) {constExponent=3})
  n11 = Constant3Vector() {constant=<rgba 0 0.1701 1 1>}
  n12 = Constant3Vector() {constant=<rgba 0 0.9759 1.5 1>}
  n13 = MaterialFunctionCall(Tile=Constant() {r=1}, DistortionAdjustment=Constant() {r=-7.1828}) {materialFunction=&MF_CylindricalProjUV}
  n14 = TextureSample(UVs=@n13.UVs, "DDX(UVs)"=@n13.UVs, "DDY(UVs)"=@n13.UVs) {
   mipValueMode=TMVM_Derivative
   samplerSource=SSM_Wrap_WorldGroupSettings
   samplerType=SAMPLERTYPE_Grayscale
   texture=&LowResBlurredNoiseNormalized
  }
  n15 = Constant3Vector() {constant=<rgba 0.8385 0 0.0037 1>}
  n16 = Constant3Vector() {constant=<rgba 1 0.0168 0 1>}
  n17 = LinearInterpolate(
   A=Multiply(A=NamedRerouteUsage(), B=LinearInterpolate(A=@n11, B=@n12, Alpha=@n14.R))
   B=Multiply(A=NamedRerouteUsage(), B=LinearInterpolate(A=@n15, B=@n16, Alpha=@n14.R))
   Alpha=NamedRerouteUsage()
  )
  n18 = LinearInterpolate(B=Constant() {r=3}, Alpha=NamedRerouteUsage()) {constA=1, constB=10}
  NamedRerouteDeclaration(None=Multiply(A=Multiply(A=@n17, B=@n18), B=LinearInterpolate(A=@n7, Alpha=@n10) {constA=0.7})) {name=EqualizerRedBlueLine, nodeColor=<rgba 0 0.5312 1 1>}
  n19 = Constant()
  NamedRerouteDeclaration(None=If(A=NamedRerouteUsage(), "A > B"=Constant() {r=1}, "A == B"=@n19, "A < B"=@n19)) {name=In_Ampltiude_LineSwitch, nodeColor=<rgba 1 0.2812 0 1>}
  NamedRerouteDeclaration(None=SphereMask(A=TextureCoordinate(), B=Constant2Vector() {g=1, r=1}, Radius=Constant() {r=0.3})) {name=Pole_Bar_Mask, nodeColor=<rgba 0.1953 1 0 1>}
  n20 = TextureObjectParameter() {
   group=TriadSpear_Pickaxe
   parameterName=Bar_Mask
   samplerType=SAMPLERTYPE_Masks
   texture=/BRCosmetics/Weapons/FORT_Melee/Pickaxe_ElegantHeist/Skins/TriadSpear/Textures/T_TriadSpear_Pickaxe_Bar_TAMask
  }
  n21 = NamedRerouteDeclaration(None=TextureSample!(UVs, Tex=@n20, Apply View MipBias).B) {name=Bar_Glow_Power, nodeColor=<rgba 1 0 0.6797 1>}
  n22 = LinearInterpolate(B=Constant() {r=0.007}, Alpha=Multiply(A=OneMinus(None=@n21), B=NamedRerouteUsage())) {constB=0.03}
  n23 = AppendVector(A=LinearInterpolate(A=Constant(), B=@n22, Alpha=NamedRerouteUsage()), B=LinearInterpolate(A=@n22, B=Constant(), Alpha=NamedRerouteUsage()))
  n24 = TextureObjectParameter() {
   group=TriadSpear_Pickaxe
   parameterName=Bar_Mask
   samplerType=SAMPLERTYPE_Masks
   texture=&T_TriadSpear_Pickaxe_Bar_TAMask
  }
  n25 = TextureSample(UVs=Add(A=TextureCoordinate(), B=Multiply(A=@n23) {constB=4}), Tex=@n24) {
   samplerType=SAMPLERTYPE_Masks
   texture=&T_TriadSpear_Pickaxe_Bar_TAMask
  }
  n26 = TextureSample(UVs=Add(A=TextureCoordinate(), B=Multiply(A=@n23) {constB=2}), Tex=@n24) {
   samplerType=SAMPLERTYPE_Masks
   texture=&T_TriadSpear_Pickaxe_Bar_TAMask
  }
  n27 = Add(A=@n25.G, B=@n26.R)
  n28 = LinearInterpolate(A=Constant3Vector() {constant=<rgba 0 0.6499 1.2 1>}, B=Constant3Vector() {constant=<rgba 0.1307 0.6444 1.5 1>}, Alpha=@n27)
  n29 = Add(A=TextureCoordinate(), B=Multiply(A=@n23) {constB=0.5})
  n30 = TextureObjectParameter() {
   group=TriadSpear_Pickaxe
   parameterName=Bar_BaseColor
   texture=/BRCosmetics/Weapons/FORT_Melee/Pickaxe_ElegantHeist/Skins/TriadSpear/Textures/T_TriadSpear_Pickaxe_Bar_E
  }
  NamedRerouteDeclaration(None=Add(
    A=Multiply(A=@n28, B=@n27)
    B=TextureSample(UVs=@n29, Tex=@n30) {texture=&T_TriadSpear_Pickaxe_Bar_E}.RGB
   )) {name=BarEmissiveColor, nodeColor=<rgba 0 1 0.0625 1>}
  NamedRerouteDeclaration(None=@n2.Emissive Color) {name=In_EmissiveColor, nodeColor=<rgba 0.4531 1 0 1>}
  n31 = NamedRerouteUsage()
  n32 = TextureSample!(UVs=@n31, Tex, "DDX(UVs)"=DDX(Value=@n31), "DDY(UVs)"=DDY(Value=@n31), Apply View MipBias) {
   mipValueMode=TMVM_Derivative
   texture=/BRCosmetics/Weapons/FORT_Melee/Pickaxe_ElegantHeist/Skins/TriadSpear/Textures/T_TriadSpear_Pickaxe_Line_TAMAsk
  }
  n33 = NamedRerouteUsage()
  n34 = TextureSample(UVs=@n33, "DDX(UVs)"=DDX(Value=@n33), "DDY(UVs)"=DDY(Value=@n33)) {
   mipValueMode=TMVM_Derivative
   samplerSource=SSM_Wrap_WorldGroupSettings
   texture=&T_TriadSpear_Pickaxe_Line_TAMAsk
  }
  NamedRerouteDeclaration(None=Max(
    A=NamedRerouteDeclaration(None=@n32.R) {name=Redline, nodeColor=<rgba 0.1953 0 1 1>}
    B=NamedRerouteDeclaration(None=@n34.G) {name=BlueLine, nodeColor=<rgba 0.125 1 0 1>}
   )) {name=LineMask, nodeColor=<rgba 1 0.4219 0 1>}
  n35 = Power(Base=NamedRerouteUsage(), Exp=Constant() {r=2}) {constExponent=4}
  n36 = MaterialFunctionCall(Tile=Constant() {r=1}, DistortionAdjustment=Constant() {r=-7.1828}) {materialFunction=&MF_CylindricalProjUV}
  n37 = TextureSample(UVs=Panner(Coordinate=@n36.UVs) {speedY=0.05}, "DDX(UVs)"=@n36.DDX, "DDY(UVs)"=@n36.DDX) {
   mipValueMode=TMVM_Derivative
   samplerSource=SSM_Wrap_WorldGroupSettings
   samplerType=SAMPLERTYPE_Grayscale
   texture=&LowResBlurredNoiseNormalized
  }
  n38 = Multiply(A=LinearInterpolate(Alpha=@n37.R) {constA=0.5}, B=Constant() {r=0.2})
  NamedRerouteDeclaration(None=If(A=NamedRerouteUsage(), "A > B"=LinearInterpolate(Alpha=@n35) {constB=0.2}, "A == B"=@n38, "A < B"=@n38)) {name=In_LineAmpltiude, nodeColor=<rgba 1 0.2812 0 1>}
  NamedRerouteDeclaration(None=MaterialFunctionCall!(InAmplitude, InSpectrum, CustomFakeSpectrum, CustomFakeAmplitude) {materialFunction=/BRCosmetics/Global/MaterialFunctions/AudioReactivity/MF_AudioReactivity_FrontendLoop}.Amplitude) {name=SourceAmplitude, nodeColor=<rgba 0.1016 0 1 1>}
  n39 = NamedRerouteUsage()
  n40 = Constant() {r=3}
  n41 = Constant() {r=0.1}
  n42 = Constant() {r=-7.1828}
  n43 = TextureCoordinate()
  n44 = TextureSample(
   UVs=Panner(Coordinate=MaterialFunctionCall(Tile=@n41, DistortionAdjustment=@n42) {materialFunction=&MF_CylindricalProjUV}.UVs) {speedX=0.2, speedY=0.7}
   "DDX(UVs)"=DDX(Value=@n43)
   "DDY(UVs)"=DDY(Value=@n43)
  ) {
   mipValueMode=TMVM_Derivative
   samplerSource=SSM_Wrap_WorldGroupSettings
   samplerType=SAMPLERTYPE_Grayscale
   texture=&LowResBlurredNoiseNormalized
  }
  n45 = Multiply(A=@n44.R, B=Constant() {r=1.2})
  NamedRerouteDeclaration(None=If(A=NamedRerouteUsage(), "A > B"=Power(Base=@n39, Exp=@n40) {constExponent=4}, "A == B"=@n45, "A < B"=@n45)) {name=IN_Amplitude_Bar, nodeColor=<rgba 0 0.6484 1 1>}
  n46 = Constant()
  NamedRerouteDeclaration(None=If(A=NamedRerouteUsage(), "A > B"=Constant() {r=1}, "A == B"=@n46, "A < B"=@n46)) {name=AudioSwitchBar, nodeColor=<rgba 0.4766 1 0 1>}
  n47 = MaterialFunctionCall(
   UVs=MaterialFunctionCall(Tile=Constant() {r=2}, DistortionAdjustment=Constant() {r=-2}) {materialFunction=&MF_CylindricalProjUV}.UVs
   "Rotation Angle (0-1)"=Constant() {r=0.22}
  ) {materialFunction=&CustomRotator}
  n48 = ComponentMask(None=Constant4Vector() {constant=<rgba 0 0.15 0 1>}.RGB) {g=true, r=true}
  n49 = LinearInterpolate(A=@n48, B=Multiply(A=@n48) {constB=0.25}, Alpha=NamedRerouteUsage())
  n50 = ComponentMask(None=Multiply(A=Panner(Coordinate=@n47.Rotated Values, Speed=@n49) {speedY=0.03}, B=Constant() {r=3.5})) {g=true}
  n51 = Saturate(None=MaterialFunctionCall(
    Input=Constant() {r=0.36}
    Input Low=ScalarParameter() {defaultValue=0.25, group=TopSide_V02_Layer Emissive_01 Reactive, parameterName=TopS Audio 01_Low_Remap, sortPriority=149}
    Input High=ScalarParameter() {defaultValue=1, group=TopSide_V02_Layer Emissive_01 Reactive, parameterName=TopS Audio 01_High_Remap, sortPriority=149}
    Target Low=Constant()
    Target High=Constant() {r=1}
   ) {materialFunction=&RemapValueRange}.Result)
  n52 = ScalarParameter() {defaultValue=0.6, group=TopSide_V02_Layer Emissive_01 Reactive, parameterName=TopS Audio 01_Falloff, sortPriority=149}
  n53 = Multiply(A=Multiply(A=Sine(None=@n50), B=Power(Base=@n51, Exp=@n52)), B=Constant() {r=-0.1131})
  n54 = Constant()
  n55 = AppendVector(A=Constant() {r=-5}, B=Constant())
  n56 = Constant()
  n57 = Multiply(
   A=LinearInterpolate(A=Constant() {r=-0.1}, B=Constant() {r=1.1}, Alpha=NamedRerouteUsage()) {constA=-0.05, constB=0.05}
   B=If(A=NamedRerouteUsage(), "A > B"=Constant() {r=1}, "A == B"=@n56, "A < B"=@n56)
  )
  NamedRerouteDeclaration(None=Multiply(A=Add(A=@n53, B=@n54), B=Multiply(A=@n55, B=@n57))) {name=AudioLineUV, nodeColor=<rgba 0 1 0.8828 1>}
  n58 = Constant() {r=0.25}
  n59 = NamedRerouteUsage()
  n60 = NamedRerouteUsage()
  n61 = Multiply(
   A=NamedRerouteUsage()
   B=LinearInterpolate(B=Constant() {r=1}, Alpha=LinearInterpolate(A=@n58, B=@n59, Alpha=@n60) {constA=1}) {constA=0.5, constB=10}
  )
  n62 = Multiply(A=@n5) {constB=0.2}
  n63 = SmoothStep(
   Min=Constant()
   Max=LinearInterpolate(A=Constant() {r=2}, B=Constant() {r=0.001}, Alpha=Saturate(None=@n59)) {constA=5, constB=0}
   Value=Saturate(None=@n6)
  )
  n64 = Saturate(None=Multiply(A=@n5, B=LinearInterpolate(Alpha=@n63) {constA=0.1, constB=1.5}))
  n65 = LinearInterpolate(Alpha=NamedRerouteUsage()) {constB=2}
  NamedRerouteDeclaration(None=Add(A=Add(A=@n61, B=@n62), B=Multiply(A=@n64, B=@n65))) {name=EqualizerBar, nodeColor=<rgba 0.8516 1 0 1>}
  n66 = NamedRerouteUsage()
  n67 = MaterialFunctionCall(Tile=Constant4Vector() {constant=<rgba 1 1 0 1>}.RGB, DistortionAdjustment=Constant() {r=-7.1828}) {materialFunction=&MF_CylindricalProjUV}
  n68 = TextureSample(UVs=Panner(Coordinate=@n67.UVs) {speedX=0.1, speedY=0.2}, "DDX(UVs)"=@n67.DDX, "DDY(UVs)"=@n67.DDX) {
   mipValueMode=TMVM_Derivative
   samplerSource=SSM_Wrap_WorldGroupSettings
   samplerType=SAMPLERTYPE_Grayscale
   texture=&LowResBlurredNoiseNormalized
  }
  n69 = LinearInterpolate(A=Multiply(A=Subtract(A=@n68.R) {constB=0.5}) {constB=2}, Alpha=NamedRerouteUsage())
  n70 = Multiply(A=@n69, B=Constant() {r=1.5})
  NamedRerouteDeclaration(None=Add(A=TextureCoordinate(), B=Multiply(A=@n66, B=@n70) {constB=1.3})) {name=RedLineUV1, nodeColor=<rgba 0.6641 0 1 1>}
  n71 = NamedRerouteUsage()
  n72 = Multiply(A=@n69, B=Constant() {r=-1.5})
  NamedRerouteDeclaration(None=Add(A=TextureCoordinate(), B=Multiply(A=@n71, B=@n72))) {name=BluelineUV1, nodeColor=<rgba 0 1 0.3438 1>}
  n73 = NamedRerouteUsage()
  n74 = Max(A=NamedRerouteUsage(), B=Multiply(A=ComponentMask(None=@n73) {b=true}) {constB=2})
  TemporalResponsivenessOutput!(None=LinearInterpolate(Alpha=Saturate(None=@n74)) {constA=0.4})
 ]
}
